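# Physics asset for the Teriyaki Fish backpack cosmetic, giving its water tube a physics setup for secondary motion.
# One capsule covers backpack_root, and four small capsules follow the dyn_water_tube bone chain.
# Constraints link each tube segment to the one before it, starting at the root.
# Tight, uneven angular limits keep swing almost planar, with slightly less freedom at the root joint and little twist throughout.
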PhysicsAsset /BRCosmetics/Accessories/FORT_Backpacks/Backpack_M_MED_Teriyaki_Fish/Meshes/Backpack_M_MED_Teriyaki_Fish_Physics
bodies {
 backpack_root [{shapeType=Capsule, length=0.5, radius=13.5021, center=<0.9419 -9.7695 0.9845>, rotation=<rot 0 -163.7401 -71.5556>}]
 dyn_water_tube_01 [{shapeType=Capsule, length=2.5, radius=3, center=<1.7877 0 0.0932>, rotation=<rot -87.0157 0.0012 70.8549>}]
 dyn_water_tube_02 [{shapeType=Capsule, length=1.75, radius=3, center=<1.4689 -0.0001 0.0766>, rotation=<rot -87.0153 0 0>}]
 dyn_water_tube_03 [{shapeType=Capsule, length=2, radius=3.25, rotation=<rot -87.0153 0 0>}]
 dyn_water_tube_04 [{shapeType=Capsule, length=2.5, radius=3.25, rotation=<rot -87.0153 0 0>}]
}
constraints [
 {bone1Name=dyn_water_tube_01, bone2Name=backpack_root, swing1LimitDegrees=1, swing2LimitDegrees=5, twistLimitDegrees=5}
 {bone1Name=dyn_water_tube_02, bone2Name=dyn_water_tube_01, swing1LimitDegrees=1, swing2LimitDegrees=10, twistLimitDegrees=5}
 {bone1Name=dyn_water_tube_03, bone2Name=dyn_water_tube_02, swing1LimitDegrees=1, swing2LimitDegrees=10, twistLimitDegrees=5}
 {bone1Name=dyn_water_tube_04, bone2Name=dyn_water_tube_03, swing1LimitDegrees=1, swing2LimitDegrees=10, twistLimitDegrees=5}
]
