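# Animation Blueprint for the Teriyaki Fish backpack cosmetic, adding secondary physics motion to its looping rig animation.
# A single-state machine plays the sequence into a RigidBody node simulated relative to backpack_root with an override physics asset.
# Chained AnimDynamics springs then move water, flower pot and seaweed bones, with a small additive offset on dyn_seaweed_a.
# AnimDynamics nodes for dyn_rope and dyn_water_tube_00 to 04 are disconnected from the output.
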AnimBlueprint /BRCosmetics/Accessories/FORT_Backpacks/Backpack_M_MED_Teriyaki_Fish/Meshes/Backpack_M_MED_Teriyaki_Fish_AnimBP
target_skeleton=&/Backpack_M_MED_Teriyaki_Fish_Skeleton
graphs [
 AnimGraph {nodes [
   Anim Dynamics - Bone: dyn_rope(Alpha=1, ExternalForce="0,0,0") {
    angularDampingOverride=0.999
    angularSpringConstant=1000
    bAngularSpring=true
    bGravityOverrideInSimSpace=true
    bOverrideAngularDamping=true
    bOverrideLinearDamping=true
    bUseGravityOverride=true
    gravityScale=0.5
    linearDampingOverride=0.999
    alphaBoolBlend {blendOption=Linear}
    boundBone {boneName=dyn_rope}
    componentAppliedLinearAccClamp=<1000 1000 1000>
    componentLinearAccScale=<1 1 1>
    componentLinearVelScale=<1 1 1>
    gravityOverride {z=-980}
    physicsBodyDefinitions [{
      sphereCollisionRadius=0
      boundBone {boneName=dyn_rope}
      boxExtents=<1 5 5>
      constraintSetup {
       angularTargetAxis=AxisY
       angularLimitsMax=<0 1 1>
       angularLimitsMin=<0 -1 -1>
       angularTarget=<0 0 0>
       linearAxesMax=<1 0 0>
       linearAxesMin=<-1 0 0>
      }
      localJointOffset=<-2 0 0>
     }]
   }
   n1 = Anim Dynamics - Bone: dyn_water_tube_00(Alpha=1, ExternalForce="0,0,0") {
    angularDampingOverride=0.999
    angularSpringConstant=1000
    bAngularSpring=true
    bGravityOverrideInSimSpace=true
    bOverrideAngularDamping=true
    bOverrideLinearDamping=true
    bUseGravityOverride=true
    linearDampingOverride=0.999
    alphaBoolBlend {blendOption=Linear}
    boundBone {boneName=dyn_water_tube_00}
    componentAppliedLinearAccClamp=<1000 1000 1000>
    componentLinearAccScale=<1 1 1>
    componentLinearVelScale=<1 1 1>
    gravityOverride {z=-980}
    physicsBodyDefinitions [{
      sphereCollisionRadius=0
      boundBone {boneName=dyn_water_tube_00}
      boxExtents=<5 10 10>
      constraintSetup {
       angularTargetAxis=AxisY
       angularLimitsMax=<0 5 5>
       angularLimitsMin=<0 0 -5>
       angularTarget=<0 0 0>
       linearAxesMax=<0.2 0 0>
       linearAxesMin=<-0.2 0 0>
      }
      localJointOffset=<-2 0 0>
     }]
   }
   n2 = Anim Dynamics - Bone: dyn_water_tube_01(ComponentPose=@n1, Alpha=1, ExternalForce="0,0,0") {
    angularDampingOverride=0.999
    angularSpringConstant=2000
    bAngularSpring=true
    bGravityOverrideInSimSpace=true
    bOverrideAngularDamping=true
    bOverrideLinearDamping=true
    bUseGravityOverride=true
    linearDampingOverride=0.999
    alphaBoolBlend {blendOption=Linear}
    boundBone {boneName=dyn_water_tube_01}
    componentAppliedLinearAccClamp=<1000 1000 1000>
    componentLinearAccScale=<1 1 1>
    componentLinearVelScale=<1 1 1>
    gravityOverride {z=-980}
    physicsBodyDefinitions [{
      sphereCollisionRadius=0
      boundBone {boneName=dyn_water_tube_01}
      boxExtents=<5 10 10>
      constraintSetup {
       angularTargetAxis=AxisY
       angularLimitsMax=<0 1 1>
       angularLimitsMin=<0 -1 -1>
       angularTarget=<0 0 0>
       linearAxesMax=<3 0 0>
       linearAxesMin=<-0.1 0 0>
      }
      localJointOffset=<-2 0 0>
     }]
   }
   n3 = Anim Dynamics - Bone: dyn_water_tube_02(ComponentPose=@n2, Alpha=1, ExternalForce="0,0,0") {
    angularDampingOverride=0.999
    angularSpringConstant=2000
    bAngularSpring=true
    bGravityOverrideInSimSpace=true
    bOverrideAngularDamping=true
    bOverrideLinearDamping=true
    bUseGravityOverride=true
    gravityScale=0.5
    linearDampingOverride=0.999
    alphaBoolBlend {blendOption=Linear}
    boundBone {boneName=dyn_water_tube_02}
    componentAppliedLinearAccClamp=<1000 1000 1000>
    componentLinearAccScale=<1 1 1>
    componentLinearVelScale=<1 1 1>
    gravityOverride {z=-980}
    physicsBodyDefinitions [{
      sphereCollisionRadius=0
      boundBone {boneName=dyn_water_tube_02}
      boxExtents=<5 10 10>
      constraintSetup {angularTargetAxis=AxisY, angularLimitsMax=<0 2 2>, angularLimitsMin=<0 -2 -2>, angularTarget=<0 0 0>}
      localJointOffset=<-1 0 0>
     }]
   }
   n4 = Anim Dynamics - Bone: dyn_water_tube_03(ComponentPose=@n3, Alpha=1, ExternalForce="0,0,0") {
    angularDampingOverride=0.999
    angularSpringConstant=2000
    bAngularSpring=true
    bGravityOverrideInSimSpace=true
    bOverrideAngularDamping=true
    bOverrideLinearDamping=true
    bUseGravityOverride=true
    linearDampingOverride=0.999
    alphaBoolBlend {blendOption=Linear}
    boundBone {boneName=dyn_water_tube_03}
    componentAppliedLinearAccClamp=<1000 1000 1000>
    componentLinearAccScale=<1 1 1>
    componentLinearVelScale=<1 1 1>
    gravityOverride {z=-980}
    physicsBodyDefinitions [{
      sphereCollisionRadius=0
      boundBone {boneName=dyn_water_tube_03}
      boxExtents=<5 10 10>
      constraintSetup {angularTargetAxis=AxisY, angularLimitsMax=<0 2 2>, angularLimitsMin=<0 -2 -2>, angularTarget=<0 0 0>, linearAxesMax=<6 0 0>}
      localJointOffset=<-1 0 0>
     }]
   }
   Anim Dynamics - Bone: dyn_water_tube_04(ComponentPose=@n4, Alpha=1, ExternalForce="0,0,0") {
    angularDampingOverride=0.999
    angularSpringConstant=2000
    bAngularSpring=true
    bGravityOverrideInSimSpace=true
    bOverrideAngularDamping=true
    bOverrideLinearDamping=true
    bUseGravityOverride=true
    gravityScale=0.5
    linearDampingOverride=0.999
    alphaBoolBlend {blendOption=Linear}
    boundBone {boneName=dyn_water_tube_04}
    componentAppliedLinearAccClamp=<1000 1000 1000>
    componentLinearAccScale=<1 1 1>
    componentLinearVelScale=<1 1 1>
    gravityOverride {z=-980}
    physicsBodyDefinitions [{
      sphereCollisionRadius=0
      boundBone {boneName=dyn_water_tube_04}
      boxExtents=<5 10 10>
      constraintSetup {angularTargetAxis=AxisY, angularLimitsMax=<0 2 2>, angularLimitsMin=<0 -2 -2>, angularTarget=<0 0 0>}
      localJointOffset=<-1 0 0>
     }]
   }
   AnimGraphNode_StateMachine_0 = New State Machine() {is=title AnimGraphNode_StateMachine}
   n5 = RigidBody(ComponentPose=Local To Component(LocalPose=@AnimGraphNode_StateMachine_0)) {
    bOverrideWorldGravity=true
    overridePhysicsAsset=&/Backpack_M_MED_Teriyaki_Fish_Physics
    simulationSpace=BaseBoneSpace
    alphaBoolBlend {blendOption=Linear}
    baseBoneRef {boneName=backpack_root}
    componentLinearAccScale=<0.5 0.5 0.5>
    componentLinearVelScale=<0.5 0.5 0.5>
    overrideWorldGravity {z=-980}
   }
   n6 = Anim Dynamics - Bone: dyn_water_tube_03_water(ComponentPose=@n5, Alpha=1, ExternalForce="0,0,0") {
    angularDampingOverride=0.999
    angularSpringConstant=2000
    bAngularSpring=true
    bGravityOverrideInSimSpace=true
    bOverrideAngularDamping=true
    bOverrideLinearDamping=true
    bUseGravityOverride=true
    linearDampingOverride=0.999
    alphaBoolBlend {blendOption=Linear}
    boundBone {boneName=dyn_water_tube_03_water}
    componentAppliedLinearAccClamp=<1000 1000 1000>
    componentLinearAccScale=<1 1 1>
    componentLinearVelScale=<1 1 1>
    gravityOverride {z=-980}
    physicsBodyDefinitions [{
      sphereCollisionRadius=0
      boundBone {boneName=dyn_water_tube_03_water}
      boxExtents=<5 10 10>
      constraintSetup {
       angularTargetAxis=AxisY
       angularLimitsMax=<0 2 2>
       angularLimitsMin=<0 -2 -2>
       angularTarget=<0 0 0>
       linearAxesMax=<6 0 0>
       linearAxesMin=<-0.5 0 0>
      }
      localJointOffset=<-1 0 0>
     }]
   }
   n7 = Anim Dynamics - Bone: dyn_water_a(ComponentPose=@n6, Alpha=1, ExternalForce="0,0,0") {
    angularDampingOverride=0.999
    angularSpringConstant=3000
    bAngularSpring=true
    bGravityOverrideInSimSpace=true
    bOverrideAngularDamping=true
    bOverrideLinearDamping=true
    bUseGravityOverride=true
    linearDampingOverride=0.999
    alphaBoolBlend {blendOption=Linear}
    boundBone {boneName=dyn_water_a}
    componentAppliedLinearAccClamp=<1000 1000 1000>
    componentLinearAccScale=<1 1 1>
    componentLinearVelScale=<1 1 1>
    gravityOverride {z=-980}
    physicsBodyDefinitions [{
      sphereCollisionRadius=0
      boundBone {boneName=dyn_water_a}
      boxExtents=<1 8 8>
      constraintSetup {
       angularTargetAxis=AxisY
       angularLimitsMax=<0 5 5>
       angularLimitsMin=<0 -5 -5>
       angularTarget=<0 0 0>
       linearAxesMax=<0.2 0 0>
       linearAxesMin=<-0.2 0 0>
      }
      localJointOffset=<-1 0 0>
     }]
   }
   n8 = Anim Dynamics - Bone: flower_pot_01(ComponentPose=@n7, Alpha=1, ExternalForce="0,0,0") {
    angularDampingOverride=0.999
    angularSpringConstant=1000
    bAngularSpring=true
    bGravityOverrideInSimSpace=true
    bOverrideAngularDamping=true
    bOverrideLinearDamping=true
    bUseGravityOverride=true
    linearDampingOverride=0.999
    alphaBoolBlend {blendOption=Linear}
    boundBone {boneName=flower_pot_01}
    componentAppliedLinearAccClamp=<1000 1000 1000>
    componentLinearAccScale=<1 1 1>
    componentLinearVelScale=<1 1 1>
    gravityOverride {z=-980}
    physicsBodyDefinitions [{
      sphereCollisionRadius=0
      boundBone {boneName=flower_pot_01}
      boxExtents=<1 10 10>
      constraintSetup {
       angularTargetAxis=AxisY
       angularLimitsMax=<0 0 10>
       angularLimitsMin=<0 0 -10>
       angularTarget=<0 0 0>
       linearAxesMax=<0.2 0.2 0.3>
       linearAxesMin=<-0.2 -0.2 -0.3>
      }
      localJointOffset=<-2 0 0>
     }]
   }
   n9 = Anim Dynamics - Bone: flower_pot_02(ComponentPose=@n8, Alpha=1, ExternalForce="0,0,0") {
    angularDampingOverride=0.999
    angularSpringConstant=1000
    bAngularSpring=true
    bGravityOverrideInSimSpace=true
    bOverrideAngularDamping=true
    bOverrideLinearDamping=true
    bUseGravityOverride=true
    linearDampingOverride=0.999
    alphaBoolBlend {blendOption=Linear}
    boundBone {boneName=flower_pot_02}
    componentAppliedLinearAccClamp=<1000 1000 1000>
    componentLinearAccScale=<1 1 1>
    componentLinearVelScale=<1 1 1>
    gravityOverride {z=-980}
    physicsBodyDefinitions [{
      sphereCollisionRadius=0
      boundBone {boneName=flower_pot_02}
      boxExtents=<1 20 20>
      constraintSetup {
       angularTargetAxis=AxisY
       angularLimitsMax=<5 5 1>
       angularLimitsMin=<-5 -5 -10>
       angularTarget=<0 0 0>
       linearAxesMax=<0.2 0.2 0.3>
       linearAxesMin=<-0.2 -0.2 -0.3>
      }
      localJointOffset=<-5 0 0>
     }]
   }
   n10 = "Transform (Modify) Bone - Bone: dyn_seaweed_a"(ComponentPose=@n9, Alpha=1, Translation="0,-0.5,0", Rotation="0,0,0", Scale="1,1,1") {translationMode=BMM_Additive, alphaBoolBlend {blendOption=Linear}, boneToModify {boneName=dyn_seaweed_a}, translation {y=-0.5}}
   n11 = Anim Dynamics - Bone: dyn_seaweed_a(ComponentPose=@n10, Alpha=1, ExternalForce="0,0,0") {
    angularDampingOverride=0.999
    angularSpringConstant=1000
    bAngularSpring=true
    bGravityOverrideInSimSpace=true
    bOverrideAngularDamping=true
    bOverrideLinearDamping=true
    bUseGravityOverride=true
    linearDampingOverride=0.999
    alphaBoolBlend {blendOption=Linear}
    boundBone {boneName=dyn_seaweed_a}
    componentAppliedLinearAccClamp=<1000 1000 1000>
    componentLinearAccScale=<1 1 1>
    componentLinearVelScale=<1 1 1>
    gravityOverride {z=-980}
    physicsBodyDefinitions [{
      sphereCollisionRadius=0
      boundBone {boneName=dyn_seaweed_a}
      boxExtents=<1 10 10>
      constraintSetup {
       angularTargetAxis=AxisY
       angularLimitsMax=<5 5 20>
       angularLimitsMin=<-5 -5 0>
       angularTarget=<0 0 0>
       linearAxesMax=<0.2 0.2 0.3>
       linearAxesMin=<-0.2 -0.2 -0.3>
      }
      localJointOffset=<5 0 0>
     }]
   }
   n12 = Anim Dynamics - Bone: dyn_seaweed_b_01(ComponentPose=@n11, Alpha=1, ExternalForce="0,0,0") {
    angularDampingOverride=0.999
    angularSpringConstant=1000
    bAngularSpring=true
    bGravityOverrideInSimSpace=true
    bOverrideAngularDamping=true
    bOverrideLinearDamping=true
    bUseGravityOverride=true
    linearDampingOverride=0.999
    alphaBoolBlend {blendOption=Linear}
    boundBone {boneName=dyn_seaweed_b_01}
    componentAppliedLinearAccClamp=<1000 1000 1000>
    componentLinearAccScale=<1 1 1>
    componentLinearVelScale=<1 1 1>
    gravityOverride {z=-980}
    physicsBodyDefinitions [{
      sphereCollisionRadius=0
      boundBone {boneName=dyn_seaweed_b_01}
      boxExtents=<1 10 10>
      constraintSetup {
       angularTargetAxis=AxisY
       angularLimitsMax=<5 5 20>
       angularLimitsMin=<-5 -5 -20>
       angularTarget=<0 0 0>
       linearAxesMax=<0.2 0.2 0.3>
       linearAxesMin=<-0.2 -0.2 -0.3>
      }
      localJointOffset=<5 0 0>
     }]
   }
   n13 = Anim Dynamics - Bone: dyn_seaweed_b_02(ComponentPose=@n12, Alpha=1, ExternalForce="0,0,0") {
    angularDampingOverride=0.999
    angularSpringConstant=1000
    bAngularSpring=true
    bGravityOverrideInSimSpace=true
    bOverrideAngularDamping=true
    bOverrideLinearDamping=true
    bUseGravityOverride=true
    linearDampingOverride=0.999
    alphaBoolBlend {blendOption=Linear}
    boundBone {boneName=dyn_seaweed_b_02}
    componentAppliedLinearAccClamp=<1000 1000 1000>
    componentLinearAccScale=<1 1 1>
    componentLinearVelScale=<1 1 1>
    gravityOverride {z=-980}
    physicsBodyDefinitions [{
      sphereCollisionRadius=0
      boundBone {boneName=dyn_seaweed_b_02}
      boxExtents=<1 10 10>
      constraintSetup {
       angularTargetAxis=AxisY
       angularLimitsMax=<5 5 20>
       angularLimitsMin=<-5 -5 -20>
       angularTarget=<0 0 0>
       linearAxesMax=<0.2 0.2 0.3>
       linearAxesMin=<-0.2 -0.2 -0.3>
      }
      localJointOffset=<5 0 0>
     }]
   }
   n14 = Anim Dynamics - Bone: dyn_seaweed_c_01(ComponentPose=@n13, Alpha=1, ExternalForce="0,0,0") {
    angularDampingOverride=0.999
    angularSpringConstant=1000
    bAngularSpring=true
    bGravityOverrideInSimSpace=true
    bOverrideAngularDamping=true
    bOverrideLinearDamping=true
    bUseGravityOverride=true
    linearDampingOverride=0.999
    alphaBoolBlend {blendOption=Linear}
    boundBone {boneName=dyn_seaweed_c_01}
    componentAppliedLinearAccClamp=<1000 1000 1000>
    componentLinearAccScale=<1 1 1>
    componentLinearVelScale=<1 1 1>
    gravityOverride {z=-980}
    physicsBodyDefinitions [{
      sphereCollisionRadius=0
      boundBone {boneName=dyn_seaweed_c_01}
      boxExtents=<1 10 10>
      constraintSetup {
       angularTargetAxis=AxisY
       angularLimitsMax=<5 5 20>
       angularLimitsMin=<-5 -5 -20>
       angularTarget=<0 0 0>
       linearAxesMax=<0.2 0.2 0.3>
       linearAxesMin=<-0.2 -0.2 -0.3>
      }
      localJointOffset=<5 0 0>
     }]
   }
   n15 = Anim Dynamics - Bone: dyn_seaweed_c_02(ComponentPose=@n14, Alpha=1, ExternalForce="0,0,0") {
    angularDampingOverride=0.999
    angularSpringConstant=1000
    bAngularSpring=true
    bGravityOverrideInSimSpace=true
    bOverrideAngularDamping=true
    bOverrideLinearDamping=true
    bUseGravityOverride=true
    linearDampingOverride=0.999
    alphaBoolBlend {blendOption=Linear}
    boundBone {boneName=dyn_seaweed_c_02}
    componentAppliedLinearAccClamp=<1000 1000 1000>
    componentLinearAccScale=<1 1 1>
    componentLinearVelScale=<1 1 1>
    gravityOverride {z=-980}
    physicsBodyDefinitions [{
      sphereCollisionRadius=0
      boundBone {boneName=dyn_seaweed_c_02}
      boxExtents=<1 10 10>
      constraintSetup {
       angularTargetAxis=AxisY
       angularLimitsMax=<5 5 20>
       angularLimitsMin=<-5 -5 -20>
       angularTarget=<0 0 0>
       linearAxesMax=<0.2 0.2 0.3>
       linearAxesMin=<-0.2 -0.2 -0.3>
      }
      localJointOffset=<5 0 0>
     }]
   }
   n16 = Anim Dynamics - Bone: dyn_seaweed_c_03(ComponentPose=@n15, Alpha=1, ExternalForce="0,0,0") {
    angularDampingOverride=0.999
    angularSpringConstant=1000
    bAngularSpring=true
    bGravityOverrideInSimSpace=true
    bOverrideAngularDamping=true
    bOverrideLinearDamping=true
    bUseGravityOverride=true
    linearDampingOverride=0.999
    alphaBoolBlend {blendOption=Linear}
    boundBone {boneName=dyn_seaweed_c_03}
    componentAppliedLinearAccClamp=<1000 1000 1000>
    componentLinearAccScale=<1 1 1>
    componentLinearVelScale=<1 1 1>
    gravityOverride {z=-980}
    physicsBodyDefinitions [{
      sphereCollisionRadius=0
      boundBone {boneName=dyn_seaweed_c_03}
      boxExtents=<1 10 10>
      constraintSetup {
       angularTargetAxis=AxisY
       angularLimitsMax=<5 5 20>
       angularLimitsMin=<-5 -5 -20>
       angularTarget=<0 0 0>
       linearAxesMax=<0.2 0.2 0.3>
       linearAxesMin=<-0.2 -0.2 -0.3>
      }
      localJointOffset=<5 0 0>
     }]
   }
   n17 = Anim Dynamics - Bone: dyn_seaweed_d(ComponentPose=@n16, Alpha=1, ExternalForce="0,0,0") {
    angularDampingOverride=0.999
    angularSpringConstant=1000
    bAngularSpring=true
    bGravityOverrideInSimSpace=true
    bOverrideAngularDamping=true
    bOverrideLinearDamping=true
    bUseGravityOverride=true
    linearDampingOverride=0.999
    alphaBoolBlend {blendOption=Linear}
    boundBone {boneName=dyn_seaweed_d}
    componentAppliedLinearAccClamp=<1000 1000 1000>
    componentLinearAccScale=<1 1 1>
    componentLinearVelScale=<1 1 1>
    gravityOverride {z=-980}
    physicsBodyDefinitions [{
      sphereCollisionRadius=0
      boundBone {boneName=dyn_seaweed_d}
      boxExtents=<1 10 10>
      constraintSetup {
       angularTargetAxis=AxisY
       angularLimitsMax=<5 5 20>
       angularLimitsMin=<-5 -5 -20>
       angularTarget=<0 0 0>
       linearAxesMax=<0.2 0.2 0.3>
       linearAxesMin=<-0.2 -0.2 -0.3>
      }
      localJointOffset=<5 0 0>
     }]
   }
   n18 = Anim Dynamics - Bone: dyn_seaweed_e(ComponentPose=@n17, Alpha=1, ExternalForce="0,0,0") {
    angularDampingOverride=0.999
    angularSpringConstant=1000
    bAngularSpring=true
    bGravityOverrideInSimSpace=true
    bOverrideAngularDamping=true
    bOverrideLinearDamping=true
    bUseGravityOverride=true
    linearDampingOverride=0.999
    alphaBoolBlend {blendOption=Linear}
    boundBone {boneName=dyn_seaweed_e}
    componentAppliedLinearAccClamp=<1000 1000 1000>
    componentLinearAccScale=<1 1 1>
    componentLinearVelScale=<1 1 1>
    gravityOverride {z=-980}
    physicsBodyDefinitions [{
      sphereCollisionRadius=0
      boundBone {boneName=dyn_seaweed_e}
      boxExtents=<1 10 10>
      constraintSetup {
       angularTargetAxis=AxisY
       angularLimitsMax=<5 5 20>
       angularLimitsMin=<-5 -5 -20>
       angularTarget=<0 0 0>
       linearAxesMax=<0.2 0.2 0.3>
       linearAxesMin=<-0.2 -0.2 -0.3>
      }
      localJointOffset=<5 0 0>
     }]
   }
   Output Pose(Result=Component To Local(ComponentPose=@n18)) {layerGroup=None}
  ]}
 New State Machine {entry=Base, path=AnimGraph.AnimGraphNode_StateMachine_0.New State Machine, ids=id, nodes [AnimStateNode_0 = Base() {is=title AnimStateNode}]}
 Base {
  path=AnimGraph.AnimGraphNode_StateMachine_0.New State Machine.AnimStateNode_0.Base
  nodes [Output Animation Pose(Result=Sequence Player 'Anim_Backpack_M_MED_Teriyaki_Fish_RIg'() {sequence=/BRCosmetics/Accessories/FORT_Backpacks/Backpack_M_MED_Teriyaki_Fish/Animation/Anim_Backpack_M_MED_Teriyaki_Fish_RIg})]
 }
 EventGraph {nodes [Event BlueprintUpdateAnimation()]}
]
variables []
